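# Translucent unlit glow material for Niagara mesh particles on the LooseCreep pickaxe cosmetic.
# SplineThicken sets strip width at base and tip from vertex color blue times a scalar parameter.
# A procedural UV mask, built from each axis times its inverse and raised to different powers, fades the edges without textures.
# Particle alpha scales that mask.
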Material /BRCosmetics/Weapons/FORT_Melee/Pickaxe_LooseCreep/FX/Materials/M_Pickaxe_LooseCreep_Glow
graph {
 properties {
  bAutomaticallySetUsageInEditor=false
  bUsedWithNiagaraMeshParticles=true
  bUsedWithStaticMesh=false
  blendMode=BLEND_Translucent
  pixelDepthOffsetMode=PDOM_Legacy
  previewMesh=/Engine/Transient.MaterialEditorMeshComponent_4
  shadingModel=MSM_Unlit
  usedShadingModels=Unlit
 }
 nodes [
  n1 = Multiply(A=VertexColor().B, B=ScalarParameter() {defaultValue=3})
  MaterialFunctionCall(WidthBase=@n1, WidthTip=@n1, Expand U or V UV Channel=StaticBool() {value=true}) {materialFunction=&SplineThicken}
  n2 = ComponentMask(None=TextureCoordinate()) {g=true}
  n3 = Multiply(A=OneMinus(None=@n2), B=@n2)
  n4 = ComponentMask(None=TextureCoordinate()) {r=true}
  n5 = Multiply(A=OneMinus(None=@n4), B=@n4)
  Multiply(A=ParticleColor().A, B=Multiply(A=Power(Base=@n3) {constExponent=0.3}, B=Power(Base=@n5) {constExponent=3}))
 ]
}
