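# Idle glow effect for the LooseCreep pickaxe, built as one Niagara emitter that bursts three blue-tinted particles that outlive their lifetime.
# Spawn groups split them into one mesh and two mirrored sprites, offset and facing opposite ways along Y, by zeroing the unused size per group.
# A curve over system age ramps alpha to a much higher level on the mesh than the sprites, while a waveform modulates RGB.
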NiagaraSystem /BRCosmetics/Weapons/FORT_Melee/Pickaxe_LooseCreep/FX/NS_Pickaxe_LooseCreep_Idle
dependencies {
 usedDataInterfaces [&NiagaraDataInterfaceCurve]
 usedDynamicInputs [&MergeHelper_Vector2Switch, &ReturnExecIndex]
 usedModules [&SystemState]
}
emitters {Glow {
  emitterUpdateScript [
   &EmitterState(
    Inactive Response=::NewEnumerator1
    Life Cycle Mode=::NewEnumerator1
    Loop Behavior=::NewEnumerator1
    Loop Duration Mode=::NewEnumerator1
    UseLoopDelay=false
   )
   &SpawnBurst_Instantaneous(Spawn Count=3, Use Spawn Probability=false, UseLoopCountLimit=false)
  ]
  particleSpawnScript [
   &FN_SetSpawnGroup(Number of Spawn Groups=3)
   &InitializeParticle(
    Color=<rgba 0.15 0.375 1.5 1>
    Lifetime=&RandomRangeFloat(Minimum=0.6, Maximum=1)
    Mesh Scale Mode=::NewEnumerator3
    Mesh Uniform Scale=&MergeHelper_FloatSwitch(0=1, 1=0, 2=0, 3=?, Index=&SpawnGroup(), Num=::NewEnumerator1)
    Position Offset=&MergeHelper_VectorSwitch(0=<0 0 0>, 1=<0.2 1.5 11>, 2=<0.2 -1.5 11>, 3=?, Index=&SpawnGroup(), Num=::NewEnumerator1)
    Sprite Size Mode=::NewEnumerator3
    Uniform Sprite Size=&MergeHelper_FloatSwitch(0=0, 1=12, 2=12, 3=?, Index=&SpawnGroup(), Num=::NewEnumerator1)
    UsePositionOffset=true
    Write Material Random=true
   )
   &SpriteFacingAndAlignment(
    Sprite Alignment=<0 0 0>
    Sprite Facing=&MergeHelper_VectorSwitch(0=<0 0 0>, 1=<0 1 0>, 2=<0 -1 0>, 3=?, Index=&SpawnGroup(), Num=::NewEnumerator1)
    WriteAlignment=true
    WriteFacing=true
   )
  ]
  particleUpdateScript [
   &ParticleState(Kill Particles When Lifetime Has Elapsed=false)
   &ScaleColor(
    Scale Alpha=&FloatFromCurve(
     FloatCurve=DataInterface() {propertyValues {Curve {keys [<at 0 0 RCIM_Cubic>,<at 1 1 RCIM_Cubic>]}}}
     CurveIndex=$Engine.System.Age
     Scale Curve=&MergeHelper_FloatSwitch(0=20, 1=0.1, 2=0.1, 3=?, Index=&SpawnGroup(), Num=::NewEnumerator1)
    )
    Scale RGB=&VectorFromFloat(Value=&Waveform(
      Waveform Count=::NewEnumerator1
      "Waveform [1]"=::NewEnumerator12
      "[1] Frequency"=6
      "[1] Amplitude Min/Max"=<0.5 1>
      "[1] Bias"=0.2
      "[2] Average"=false
      "[2] Input Position"=0
      "[2] Phase"=$Emitter.Age
      "[2] Phase Scale"=1
      "[2] Compound Frequency"=2
      "[2] Frequency"=1.2
      "[2] Amplitude Min/Max"=<0.5 1>
      "[2] Scale"=1
      "[2] Bias"=0
      "[2] Exponent"=1
      "[2] Pulse Duration"=0.1
      "[2] Clip Triangle Min/Max"=<0 0>
      "[2] Random Seed"=0
     ))
    ScaleA=true
    ScaleRGB=true
   )
  ]
  rendererClasses [&NiagaraMeshRendererProperties, &NiagaraSpriteRendererProperties]
  renderers [
   {rendererClass=&NiagaraMeshRendererProperties, rendererIndex=0}
   {rendererClass=&NiagaraSpriteRendererProperties, rendererIndex=1}
  ]
 }}
summary {
 systemName=NS_Pickaxe_LooseCreep_Idle
 emitters [{emitterName=Glow, rendererClasses [&NiagaraMeshRendererProperties, &NiagaraSpriteRendererProperties]}]
 userVariables [
  User.bIsFrontEnd {type=NiagaraBool}
  User.bIsFrontEndPreview {type=NiagaraBool}
 ]
}
